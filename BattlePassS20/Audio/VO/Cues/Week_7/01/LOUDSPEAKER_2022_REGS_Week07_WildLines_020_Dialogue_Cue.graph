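# Sound cue for a Battle Pass Season 20 loudspeaker voice line, wild line 020 from the Week 7 quest dialogue set.
# A single dialogue player node plays the Loudspeaker dialogue wave with the Loudspeaker voice as speaker and no listed targets.
# The cue uses the narrator VO sound class, the shared Season 20 quest attenuation and a boosted volume multiplier.
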SoundCue /BattlePassS20/Audio/VO/Cues/Week_7/01/LOUDSPEAKER_2022_REGS_Week07_WildLines_020_Dialogue_Cue
root=DialoguePlayer_0
nodes [{
  id=DialoguePlayer_0
  properties {dialogueWaveParameter {
    dialogueWave=/BattlePassS20/Audio/VO/Waves/Loudspeaker/LOUDSPEAKER_2022_REGS_Week07_WildLines_020_Dialogue
    context {speaker=/BattlePassS20/Audio/VO/Loudspeaker_DV, targets []}
   }}
 }]
properties {
 attenuationSettings=/BattlePassS20/Audio/VO/S20_Quest_Att
 maxDistance=2097152
 soundClassObject=/Game/Sounds/FortSoundClassesAndModes/Fort_VO_Narrator
 volumeMultiplier=2
}
